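# Translucent unlit Niagara ribbon material for the LightningDragon pickaxe's ink animation trail.
# Particle color drives emissive through MF_FX_EmissiveAdjustments.
# A U-scalable texture mask, multiplied by particle alpha and eroded by subtracting the Dissolve dynamic parameter, forms opacity, with an optional fade toward the high-U end.
# A shading-path switch applies depth fade on default and deferred paths, skipping it on forward and mobile.
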Material /BRCosmetics/Weapons/FORT_Melee/Pickaxe_LightningDragon/FX/M_Pickaxe_LightningDragon_Inks_AnimTrail
graph {
 properties {bUsedWithNiagaraRibbons=true, blendMode=BLEND_Translucent, pixelDepthOffsetMode=PDOM_Legacy, shadingModel=MSM_Unlit, usedShadingModels=Unlit}
 nodes [
  n1 = ParticleColor()
  MaterialFunctionCall(Emissive=@n1.RGB) {materialFunction=&MF_FX_EmissiveAdjustments}
  n2 = TextureCoordinate()
  n3 = AppendVector(A=ScalarParameter() {defaultValue=1, parameterName=U_Scale}, B=Constant() {r=1})
  n4 = StaticComponentMaskParameter(None=TextureSampleParameter2D(UVs=Multiply(A=@n2, B=@n3)) {parameterName=Default Tex, texture=&/T_Inks_AnimTrail}.RGB) {defaultR=true, parameterName=RGB Mask}
  n5 = ComponentMask(None=TextureCoordinate()) {r=true}
  n6 = Multiply(A=Subtract(A=OneMinus(None=@n5)) {constB=0.2}) {constB=10}
  n7 = Multiply(
   A=Subtract(A=Multiply(A=@n1.A, B=@n4), B=DynamicParameter() {defaultValue=<rgba 0 0 1 1>, paramNames [Dissolve,Lerp_Tex,Param3,Param4]}.Dissolve)
   B=LinearInterpolate(B=Saturate(None=@n6), Alpha=ScalarParameter() {parameterName=Use Edge Mask}) {constA=1}
  )
  n8 = DepthFade(Opacity=@n7, FadeDistance=ScalarParameter() {defaultValue=20, parameterName=FadeDistance}) {fadeDistanceDefault=20}
  ShadingPathSwitch(Default=@n8, Deferred=@n8, Forward=@n7, Mobile=@n7)
 ]
}
